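# M_3DHB is a translucent, unlit master material for hologram effects on static or skeletal meshes and Niagara particles.
# MF_3DHB_Base supplies color, opacity, WPO and normal through named reroutes, while MF_3DHB_HologramFX applies hologram, scanline and glitch effects.
# Dynamic parameters provide a random time offset and mesh scale, though MeshScale also drives the CustomFlicker input.
# EnableParticleColorControl blends in particle color and alpha, and MF_FX_EmissiveAdjustments processes the emissive.
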Material /Game/Tools/3DHB/Shader/MasterLib/M_3DHB
graph {
 properties {
  allowTranslucentCustomDepthWrites=true
  bUsedWithMeshDeformer=true
  bUsedWithNiagaraMeshParticles=true
  bUsedWithNiagaraSprites=true
  bUsedWithSkeletalMesh=true
  blendMode=BLEND_Translucent
  ditherOpacityMask=true
  opacityMaskClipValue=0.05
  pixelDepthOffsetMode=PDOM_Legacy
  shadingModel=MSM_Unlit
  usedShadingModels=Unlit
 }
 nodes [
  n1 = DynamicParameter() {paramNames [Random, MeshScale, CustomFlicker, Param4]}
  n2 = Add(A=Time() {bOverride_Period=true, period=6000}, B=@n1.Random)
  n3 = MaterialFunctionCall() {
   desc=Allows custom world space direction vector for scanlines. Defaults to off
   materialFunction=/Game/Tools/3DHB/Shader/Base/MF_3DHB_HologramFX_UVForScanline
  }
  n4 = MaterialFunctionCall!(
   Time=@n2
   MeshScaleIn=@n1.MeshScale
   UVForScanline=@n3.UVForScanline
   CustomFlicker=@n1.MeshScale
   UseCustomFlicker=ScalarParameter() {group=Base_HoloFX_Glitch, parameterName=Glitch_UseCustomFlicker}
   UVForStroke
   UVForGlitch
   BaseUVIn
   EmissiveIn=NamedRerouteUsage()
   OpacityIn=NamedRerouteUsage()
   WPOIn=NamedRerouteUsage()
  ) {materialFunction=/Game/Tools/3DHB/Shader/Base/MF_3DHB_HologramFX}
  n5 = ParticleColor()
  n6 = ScalarParameter() {group=TOD, parameterName=EnableParticleColorControl}
  LinearInterpolate(A=@n4.OpacityOut, B=Multiply(A=@n5.A, B=@n4.OpacityOut), Alpha=@n6)
  MaterialFunctionCall(Emissive=LinearInterpolate(A=@n4.EmissiveOut, B=Multiply(A=@n5.RGB, B=@n4.EmissiveOut), Alpha=@n6)) {materialFunction=&MF_FX_EmissiveAdjustments}
  n7 = MaterialFunctionCall!(MeshScaleIn=@n1.MeshScale) {materialFunction=/Game/Tools/3DHB/Shader/Base/MF_3DHB_Base}
  NamedRerouteDeclaration(None=ComponentMask(None=@n7.OriginDiffuse) {b=true, g=true, r=true}) {name=Base_Emissive, nodeColor=<rgba 0.5703 1 0 1>}
  NamedRerouteDeclaration(None=ComponentMask(None=@n7.OriginDiffuse) {a=true}) {name=Base_Opacity, nodeColor=<rgba 0.5703 1 0 1>}
  NamedRerouteDeclaration(None=@n7.OriginWPO) {name=Base_WPO, nodeColor=<rgba 0.5703 1 0 1>}
  NamedRerouteDeclaration(None=@n7.OriginNormal) {name=Base_Normal, nodeColor=<rgba 0.5703 1 0 1>}
  NamedRerouteDeclaration(None=NamedRerouteUsage()) {name=Post_Emissive, nodeColor=<rgba 0.5703 1 0 1>}
  NamedRerouteDeclaration(None=NamedRerouteUsage()) {name=Post_Opacity, nodeColor=<rgba 0.5703 1 0 1>}
  NamedRerouteDeclaration(None=NamedRerouteUsage()) {name=Post_WPO, nodeColor=<rgba 0.5703 1 0 1>}
  NamedRerouteDeclaration(None=@n2) {name=RandomTime, nodeColor=<rgba 0 1 0.7422 1>}
  NamedRerouteUsage()
 ]
}
